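; Parent Blueprint for a Playgrounds concert spotlight prop, apparently a set of four, that manages the material on its light beam mesh.
; The construction script applies the light ray override material to the first slot of LightBeams, only where cosmetic events can execute.
; GetOverrideMeshMaterial returns that same material when asked about the LightBeams component and none for any other mesh.
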(fn GetOverrideMeshMaterial
  (params (MeshComp :type Static%20Mesh%20Component%20Object%20Reference) (MatIdx :type Integer))
  (if (Utilities|Equal(Object) MeshComp (Variables|Default|GetLightBeams))
    (return (Variables|Default|GetLightRayMaterialOverride))
    (else
      (return 0))))

(fn ConstructionScript
  (Networking|CanExecuteCosmeticEvents
    (:True
      (bind _lightraymaterialoverride (Variables|Default|GetLightRayMaterialOverride)
        (:then
          (Rendering|Material|SetMaterial (Variables|Default|GetLightBeams) 0 _lightraymaterialoverride))
        (:else)))))
