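; Parent Blueprint for glow stick light props in Playgrounds, pairing a colored point light with a glow mesh.
; Its construction script, only where cosmetic events can run, applies the light color and any valid glow material override.
; It also switches the glow mesh to absolute rotation and pins it to a fixed world orientation.
; GetOverrideMeshMaterial reports that override for the glow mesh, ignoring the material index, and returns none otherwise.
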(fn GetOverrideMeshMaterial
  (params (MeshComp :type Static%20Mesh%20Component%20Object%20Reference) (MatIdx :type Integer))
  (if (Utilities|Equal(Object) MeshComp (Variables|Default|GetGlowMesh))
    (bind _glowmaterialoverride (Variables|Default|GetGlowMaterialOverride)
      (:then
        (return _glowmaterialoverride))
      (:else
        _))
    (else
      _))
  (return 0))

(fn ConstructionScript
  (Networking|CanExecuteCosmeticEvents
    (:True
      (Rendering|Components|Light|SetLightColor (Variables|Default|GetPointLight) (Variables|Default|GetLightColor) true)
      (Variables|/Script/Engine.SceneComponent|SetbAbsoluteRotation true (Variables|Default|GetGlowMesh))
      (Transformation|SetWorldRotation (Variables|Default|GetGlowMesh) "0,-90.000000, 0" false false)
      (bind _glowmaterialoverride (Variables|Default|GetGlowMaterialOverride)
        (:then
          (Rendering|Material|SetMaterial (Variables|Default|GetGlowMesh) 0 _glowmaterialoverride))
        (:else)))))
