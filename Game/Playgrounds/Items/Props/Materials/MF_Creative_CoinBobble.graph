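# Material function that produces a vertical bobbing World Position Offset, named for coin props in Creative.
# Time scaled by the TimeScale input drives a Sine, which is the alpha of a lerp between the MaxHeight and MinHeight inputs.
# Both height inputs default to vector parameters, BobHeightMax and BobHeightMin.
# The sine is not remapped to 0–1, so its negative half extrapolates the offset beyond MaxHeight.
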MaterialFunction /Game/Playgrounds/Items/Props/Materials/MF_Creative_CoinBobble
graph {nodes [
  n1 = VectorParameter() {parameterName=BobHeightMax, defaultValue=<rgba 0 0 10 0>}
  n2 = VectorParameter() {parameterName=BobHeightMin, defaultValue=<rgba 0 0 -2 0>}
  n3 = Multiply(A=Time(), B=FunctionInput(Preview=Constant() {r=0.4}) {bUsePreviewValueAsDefault=true, inputName=TimeScale, inputType=FunctionInput_Scalar})
  FunctionOutput(None=LinearInterpolate(
    A=FunctionInput(Preview=@n1.RGB) {bUsePreviewValueAsDefault=true, inputName=MaxHeight}
    B=FunctionInput(Preview=@n2.RGB) {bUsePreviewValueAsDefault=true, inputName=MinHeight, sortPriority=1}
    Alpha=Sine(None=@n3)
   )) {outputName=WorldPositionOffset}
 ]}
