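# Surface material for a puzzle-piece coin prop in Playgrounds, using Llama Puzzle textures by default.
# Parameterized diffuse, packed mask (specular, metallic, roughness), normal and emissive maps, most with scale or color multipliers, fill the material attributes.
# MF_Creative_CoinBobble drives World Position Offset to animate the coin in the shader.
# The combined attributes then pass through MF_CreativeMoveTool, the Creative move-tool material function.
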Material /Game/Playgrounds/Items/Props/Materials/M_Coin_PuzzlePiece
graph {
 properties {bUseMaterialAttributes=true, pixelDepthOffsetMode=PDOM_Legacy, usedShadingModels=Default Lit}
 nodes [
  n1 = TextureSampleParameter2D() {parameterName=Masks, samplerType=SAMPLERTYPE_Masks, texture=/Game/Athena/Items/QuestInteractables/Llama_Puzzle/Textures/T_Llama_Puzzle_S}
  n2 = TextureSampleParameter2D() {parameterName=Normal, samplerType=SAMPLERTYPE_Normal, texture=/Game/Athena/Items/QuestInteractables/Llama_Puzzle/Textures/T_Llama_Puzzle_N}
  n3 = SetMaterialAttributes!(
   MaterialAttributes
   Base Color=Multiply(
    A=VectorParameter() {parameterName=DiffuseColorScale, defaultValue=<rgba 1 1 1 0>}.RGB
    B=TextureSampleParameter2D() {parameterName=Diffuse, texture=/Game/Athena/Items/QuestInteractables/Llama_Puzzle/Textures/T_Llama_Puzzle_Unsolved_D}.RGB
   )
   Metallic=Multiply(A=@n1.G, B=ScalarParameter() {defaultValue=1, parameterName=MetallicScale})
   Specular=Multiply(A=@n1.R, B=ScalarParameter() {defaultValue=1, parameterName=SpecularScale})
   Roughness=Multiply(A=@n1.B, B=ScalarParameter() {defaultValue=1, parameterName=RoughnessScale})
   Normal=@n2.RGB
   World Position Offset=MaterialFunctionCall!(TimeScale, MaxHeight, MinHeight) {materialFunction=&/MF_Creative_CoinBobble}.WorldPositionOffset
   Emissive Color=Multiply(
    A=TextureSampleParameter2D() {parameterName=Emissive, texture=&T_BlacksRGB}.RGB
    B=VectorParameter() {parameterName=EmissiveColor}.RGB
   )
  ) {attributeSetTypes [
    #1
    #2
    #3
    #4
    #5
    #6
    #7
   ]}
  MaterialFunctionCall(MA In=@n3) {materialFunction=&MF_CreativeMoveTool}
 ]
}
