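# MF_Scale2D is a material function in the Playgrounds props materials that scales a 2D vector by a scalar.
# It takes a Vector2 input named XY and a scalar input named Scalar.
# It outputs XY divided by Scalar, so a larger Scalar yields smaller coordinate values.
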MaterialFunction /Game/Playgrounds/Items/Props/Materials/MF_Scale2D
graph {nodes [FunctionOutput(None=Divide(A=FunctionInput() {inputName=XY, inputType=FunctionInput_Vector2}, B=FunctionInput() {inputName=Scalar, inputType=FunctionInput_Scalar}))]}
