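# MF_Translate2D is a material function in the Props materials folder that translates a 2D coordinate relative to an origin.
# It takes two Vector2 inputs, XY and Origin, and outputs XY minus Origin.
# The result is the coordinate re-expressed with Origin as the zero point.
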MaterialFunction /Game/Playgrounds/Items/Props/Materials/MF_Translate2D
graph {nodes [FunctionOutput(None=Subtract(A=FunctionInput() {inputName=XY, inputType=FunctionInput_Vector2}, B=FunctionInput() {inputName=Origin, inputType=FunctionInput_Vector2}))]}
