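# Translucent, lit hexagon decal material for props, based on the T_Emblem_HexTile_01 emblem texture.
# Translate and scale functions resize the UVs about their center via Scale_UVs, feeding a clamped texture sample parameter.
# The sample is multiplied by Color_Tint and Color_Tint_Intensity, and separately inverted with OneMinus.
# The graph shows no links from these results, or from two loose constants, to material outputs.
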Material /Game/Playgrounds/Items/Props/Materials/M_Decal_Hexagon_01
graph {
 properties {blendMode=BLEND_Translucent, pixelDepthOffsetMode=PDOM_Legacy, translucencyLightingMode=TLM_Surface, usedShadingModels=Default Lit}
 nodes [
  Constant() {r=0.1}
  Constant() {r=0.5}
  n1 = Constant2Vector() {g=0.5, r=0.5}
  n2 = ScalarParameter() {defaultValue=1, parameterName=Scale_UVs}
  n3 = MaterialFunctionCall!(Origin=@n1, XY=TextureCoordinate()) {materialFunction=&/MF_Translate2D}
  n4 = TextureSampleParameter2D(UVs=MaterialFunctionCall(
    Origin=Multiply(A=@n1) {constB=-1}
    XY=MaterialFunctionCall!(Scalar=@n2, XY=@n3.Result) {materialFunction=&/MF_Scale2D}.Result
   ) {materialFunction=&MF_Translate2D}.Result) {parameterName=Texture Sample, samplerSource=SSM_Clamp_WorldGroupSettings, texture=/Game/Playgrounds/Items/Props/Textures/T_Emblem_HexTile_01}
  Multiply(
   A=Multiply(A=VectorParameter() {parameterName=Color_Tint, defaultValue=<rgba 0 0 0 1>}.RGB, B=@n4.RGB)
   B=ScalarParameter() {parameterName=Color_Tint_Intensity}
  )
  OneMinus(None=@n4.RGB)
 ]
}
