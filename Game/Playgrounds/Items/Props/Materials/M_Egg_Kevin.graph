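# Two-sided lit material for the "Kevin" egg prop, a dark purple surface with parallax sparkles and a pulsing inner glow with runes.
# UV2 edge masks height-lerped with cube noise shape the surface.
# A camera-ray sphere thickness drives the glow, layered with bump-offset DFHeightMask runes and scaled by RuneActive, Zap and Zap2.
# It adds an optional world-height SinkingGradient, cheaper low-quality branches and simplified HLOD proxy values.
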Material /Game/Playgrounds/Items/Props/Materials/M_Egg_Kevin
graph {
 properties {pixelDepthOffsetMode=PDOM_Legacy, twoSided=true, usedShadingModels=Default Lit}
 nodes [
  ActorPositionWS()
  n1 = VertexColor()
  n2 = TwoSidedSign()
  n3 = Constant() {r=0.2}
  n4 = Constant() {r=0.1}
  n5 = Constant() {r=2}
  n6 = TextureCoordinate() {coordinateIndex=2}
  n7 = Constant2Vector() {g=0.5, r=0.5}
  n8 = MaterialFunctionCall(Float2=Abs(None=Subtract(A=@n6, B=@n7))) {materialFunction=&BreakOutFloat2Components}
  n9 = Max(A=@n8.R, B=@n8.R)
  n10 = Power(Base=Saturate(None=Multiply(A=@n9, B=@n5)), Exp=Constant() {r=1.8})
  n11 = Multiply(A=Constant() {r=3.5}, B=TextureCoordinate())
  n12 = TextureSample(UVs=@n11) {samplerType=SAMPLERTYPE_Masks, texture=/Game/Global/Textures/Default/Noise/T_CubeNoise}
  n13 = Saturate(None=Add(
    A=MaterialFunctionCall(Transition Phase=@n10, Height Texture=@n12.R) {materialFunction=&HeightLerp}.Lerp Alpha No Contrast
    B=@n12.R
   ))
  n14 = Multiply(A=@n1.R, B=Multiply(A=@n2, B=LinearInterpolate(A=@n3, B=@n4, Alpha=@n13)))
  Clamp(None=@n14, Max=@n14)
  n15 = VectorParameter() {parameterName=BaseDiffuse, defaultValue=<rgba 0.015 0.0013 0.1 1>}
  n16 = TextureCoordinate()
  n17 = Multiply(A=Constant() {r=14}, B=@n16)
  n18 = Constant() {r=-2}
  n19 = Constant() {r=0.02}
  n20 = Multiply(A=@n16, B=Constant() {r=28})
  n21 = Constant() {r=0.2}
  n22 = Multiply(
   A=TextureSample(UVs=BumpOffset(Coordinate=@n17, Height=@n18, HeightRatioInput=@n19)) {samplerType=SAMPLERTYPE_Masks, texture=&T_CubeNoise}.B
   B=TextureSample(UVs=BumpOffset(Coordinate=@n20, Height=@n18, HeightRatioInput=@n21)) {samplerType=SAMPLERTYPE_Masks, texture=&T_CubeNoise}.B
  )
  n23 = Constant() {r=7}
  n24 = Multiply(
   A=VectorParameter() {parameterName=SparklesColor, defaultValue=<rgba 0.5857 0.2 1 1>}.RGB
   B=Multiply(A=Power(Base=@n22, Exp=@n23), B=Constant() {r=60})
  )
  n25 = TextureSample!(
   UVs=BumpOffset(Coordinate=@n11, Height=Constant() {r=0.58}, HeightRatioInput=Constant() {r=-0.5})
   Tex
   Bias=ScalarParameter() {defaultValue=5, parameterName=mipbias1}
   Apply View MipBias
  ) {mipValueMode=TMVM_MipBias, samplerType=SAMPLERTYPE_Masks, texture=&T_CubeNoise}
  n26 = Constant() {r=3}
  n27 = Saturate(None=Add(A=Power(Base=@n25.R, Exp=@n26), B=Constant() {r=-0.05}))
  n28 = Saturate(None=Multiply(A=OneMinus(None=@n9), B=Constant() {r=1.7}))
  n29 = Constant() {r=30}
  n30 = Add(A=@n15.RGB, B=Multiply(A=Multiply(A=@n24, B=@n27), B=Power(Base=@n28, Exp=@n29)))
  n31 = QualitySwitch(Default=@n30, Low=@n15.RGB, High=@n30)
  n32 = VectorParameter() {parameterName=EdgeDiffuse, defaultValue=<rgba 0.006 0.0005 0.04 1>}
  n33 = Multiply(A=@n9, B=Constant() {r=1.95})
  n34 = MaterialFunctionCall(Transition Phase=Power(Base=Saturate(None=@n33), Exp=Constant() {r=5}), Height Texture=@n12.R) {materialFunction=&HeightLerp}
  n35 = Multiply(A=@n1.R, B=Multiply(A=LinearInterpolate(A=@n31, B=@n32.RGB, Alpha=@n34.Lerp Alpha No Contrast), B=@n2))
  Clamp(None=@n35, Max=@n35)
  n36 = WorldPosition()
  n37 = ScalarParameter() {defaultValue=-300, group=SinkingGradient, parameterName=SinkGradOffset}
  n38 = Multiply(
   A=Multiply(A=Add(A=@n36.XYZ, B=@n37), B=Constant() {r=-1})
   B=ScalarParameter() {defaultValue=0.0009, group=SinkingGradient, parameterName=SinkGradSize}
  )
  n39 = Power(
   Base=ComponentMask(None=Saturate(None=@n38)) {b=true}
   Exp=ScalarParameter() {defaultValue=13, group=SinkingGradient, parameterName=SinkGradPower}
  )
  n40 = VectorParameter() {parameterName=GlowColor1, defaultValue=<rgba 0.012 0.0011 0.08 1>}
  n41 = VectorParameter() {parameterName=GlowColor2, defaultValue=<rgba 0.0356 0.001 0.2 1>}
  n42 = VectorParameter() {parameterName=GlowColor3, defaultValue=<rgba 0.1264 0 0.3 1>}
  n43 = VectorParameter() {parameterName=GlowColor4, defaultValue=<rgba 0.3583 0 0.5 1>}
  n44 = MaterialFunctionCall(Value=Multiply(A=Time(), B=ScalarParameter() {defaultValue=0.5, parameterName=PulseSpeed_3})) {materialFunction=&LinearSine}
  n45 = Multiply(
   A=LinearInterpolate(A=Constant() {r=1200}, B=Constant() {r=1400}, Alpha=@n44."Rounded Linear Sine ")
   B=ScalarParameter() {defaultValue=1, parameterName=RadiusMult}
  )
  n46 = Multiply(A=@n45, B=@n45) {desc="r^2"}
  n47 = ObjectPositionWS()
  n48 = Add(A=VectorParameter() {parameterName=offz, defaultValue=<rgba 0 0 -400 1>}.RGB, B=@n47)
  n49 = ScalarParameter() {parameterName=RuneActive}
  n50 = Subtract(
   A=CameraPositionWS()
   B=Add(A=VectorParameter() {desc=For zapping players, parameterName=SphereOffset}.RGB, B=LinearInterpolate(A=@n47, B=@n48, Alpha=@n49))
  )
  n51 = Distance(A=@n50, B=Constant())
  n52 = DotProduct(A=@n50, B=CameraVectorWS())
  n53 = Subtract(A=Multiply(A=@n51, B=@n51), B=Multiply(A=@n52, B=@n52))
  n54 = SquareRoot(None=Max(A=Subtract(A=@n46, B=@n53)) {constB=0})
  n55 = Multiply(A=Divide(A=Multiply(A=@n54) {constB=2}, B=Multiply(A=@n45) {constB=2}), B=ScalarParameter() {defaultValue=0.87, parameterName=gr334})
  n56 = Saturate(None=Add(A=@n12.R, B=ScalarParameter() {defaultValue=0.25, parameterName=4444235}))
  n57 = ScalarParameter() {defaultValue=2.75, parameterName=2222}
  n58 = ScalarParameter() {defaultValue=1.5, parameterName=434233}
  n59 = Multiply(A=@n56, B=Power(Base=@n25.G, Exp=LinearInterpolate(A=@n57, B=@n58, Alpha=@n44."Rounded Linear Sine ")))
  n60 = SphereMask(A=TextureCoordinate() {coordinateIndex=2}, B=Constant() {r=0.5}) {attenuationRadius=0.7, hardnessPercent=30}
  n61 = MaterialFunctionCall(Transition Phase=@n55, Height Texture=Multiply(A=@n60, B=QualitySwitch(Default=@n59, Low=@n56, High=@n59))) {materialFunction=&HeightLerp}
  n62 = Multiply(
   A=Power(Base=@n55, Exp=ScalarParameter() {defaultValue=20, parameterName=gpow3})
   B=ScalarParameter() {defaultValue=75, parameterName=ActiveHotspot}
  )
  n63 = Power(Base=@n25.R, Exp=Constant() {r=3})
  n64 = Multiply(A=@n62, B=@n63)
  n65 = Multiply(A=QualitySwitch(Default=@n64, Low=@n62, High=@n64), B=@n60)
  n66 = Multiply(
   A=Add(
    A=MaterialFunctionCall(1=@n40.RGB, 2=@n41.RGB, 3=@n42.RGB, 4=@n43.RGB, A=@n61.Lerp Alpha No Contrast) {materialFunction=&Lerp_Multiple_Float3}.Lerp 4 Inputs
    B=Multiply(A=@n43.RGB, B=@n65)
   )
   B=@n60
  )
  n67 = ScalarParameter() {defaultValue=0.25, parameterName=deactivated_1}
  n68 = ScalarParameter() {defaultValue=0.05, parameterName=activated_1}
  n69 = Saturate(None=Add(A=LinearInterpolate(A=@n67, B=@n68, Alpha=@n49), B=@n63))
  n70 = Multiply(A=@n60, B=@n61.Lerp Alpha No Contrast)
  n71 = ScalarParameter() {defaultValue=2.5, parameterName=gpow3_1}
  n72 = MaterialFunctionCall(Float2=Multiply(A=ScalarParameter() {defaultValue=75, parameterName=BB_15}, B=TextureCoordinate() {uTiling=0.95, vTiling=0.75})) {materialFunction=&BreakOutFloat2Components}
  n73 = MaterialFunctionCall(X=Add(A=ScalarParameter() {parameterName=BB_19}, B=@n72.R), Y=Add(A=@n72.G, B=ScalarParameter() {defaultValue=0.5, parameterName=BB_20})) {materialFunction=&MakeFloat2}
  n74 = Constant() {r=-1}
  n75 = Constant() {r=1}
  n76 = TextureSample!(UVs=@n73.Result, Tex, Apply View MipBias) {samplerType=SAMPLERTYPE_LinearColor, texture=/Game/Athena/Prototype/Blueprints/Cube/Materials/DFHeightMask}
  n77 = BumpOffset(
   Coordinate=@n73.Result
   Height=Multiply(A=LinearInterpolate(A=@n74, B=@n75, Alpha=@n76.G) {constA=-1}, B=ScalarParameter() {defaultValue=1, parameterName=qw1_2})
   HeightRatioInput=ScalarParameter() {defaultValue=0.06, parameterName=BB_16}
  )
  n78 = ScalarParameter() {defaultValue=10, parameterName=G_8}
  n79 = ScalarParameter() {defaultValue=2, parameterName=G_7}
  n80 = MaterialFunctionCall(Float2=Multiply(A=ScalarParameter() {defaultValue=19, parameterName=BB_3}, B=TextureCoordinate() {vTiling=0.75})) {materialFunction=&BreakOutFloat2Components}
  n81 = MaterialFunctionCall(
   X=Add(A=ScalarParameter() {defaultValue=0.1, parameterName=BB_13}, B=@n80.R)
   Y=Add(A=@n80.G, B=ScalarParameter() {defaultValue=0.05, parameterName=BB_14})
  ) {materialFunction=&MakeFloat2}
  n82 = LinearInterpolate(
   A=Constant() {r=-1}
   B=Constant() {r=1}
   Alpha=TextureSample(UVs=@n81.Result) {samplerType=SAMPLERTYPE_LinearColor, texture=&DFHeightMask}.G
  ) {constA=-1}
  n83 = ScalarParameter() {defaultValue=0.5, parameterName=qw1_1}
  n84 = TextureSample(UVs=BumpOffset(Coordinate=@n81.Result, Height=Multiply(A=@n82, B=@n83), HeightRatioInput=ScalarParameter() {defaultValue=0.15, parameterName=BB_8})) {samplerType=SAMPLERTYPE_LinearColor, texture=&DFHeightMask}
  n85 = ScalarParameter() {defaultValue=1, parameterName=G_5}
  n86 = LinearInterpolate(
   A=Power(
    Base=TextureSample(UVs=@n77) {samplerType=SAMPLERTYPE_LinearColor, texture=&DFHeightMask}.G
    Exp=LinearInterpolate(A=@n78, B=@n79, Alpha=@n44."Rounded Linear Sine ")
   )
   B=Power(Base=@n84.G, Exp=ScalarParameter() {defaultValue=20, parameterName=G_2})
   Alpha=Saturate(None=Multiply(A=@n84.G, B=@n85))
  )
  n87 = Multiply(A=ScalarParameter() {defaultValue=32, parameterName=BB_2}, B=TextureCoordinate())
  n88 = BumpOffset(
   Coordinate=@n87
   Height=LinearInterpolate(
    A=ScalarParameter() {defaultValue=-1, parameterName=qw1_8}
    B=ScalarParameter() {defaultValue=1, parameterName=qw1_9}
    Alpha=TextureSample(UVs=@n87) {samplerType=SAMPLERTYPE_LinearColor, texture=&DFHeightMask}.R
   )
   HeightRatioInput=ScalarParameter() {defaultValue=0.06, parameterName=BB_4}
  )
  n89 = TextureSample(UVs=QualitySwitch(Default=@n88, Low=@n87, High=@n88)) {samplerType=SAMPLERTYPE_LinearColor, texture=&DFHeightMask}
  n90 = ScalarParameter() {defaultValue=1.5, parameterName=G_6}
  n91 = ScalarParameter() {defaultValue=15, parameterName=G_1}
  n92 = Multiply(
   A=Power(Base=@n89.R, Exp=LinearInterpolate(A=@n90, B=@n91, Alpha=@n44."Rounded Linear Sine "))
   B=ScalarParameter() {defaultValue=5, parameterName=G_9}
  )
  n93 = Panner(
   Coordinate=Multiply(A=ScalarParameter() {defaultValue=2, parameterName=BB_1}, B=TextureCoordinate())
   Time=Multiply(A=ScalarParameter() {defaultValue=0.0135, parameterName=BB_7}, B=Time())
  ) {speedX=1}
  n94 = Saturate(None=LinearInterpolate(
    A=ScalarParameter() {defaultValue=1, parameterName=qw1_3}
    B=ScalarParameter() {defaultValue=-1, parameterName=qw1_4}
    Alpha=TextureSample(UVs=@n93) {samplerType=SAMPLERTYPE_LinearColor, texture=&DFHeightMask}.B
   ))
  n95 = Saturate(None=Multiply(A=Multiply(A=@n89.R, B=@n94), B=ScalarParameter() {defaultValue=1.25, parameterName=G_3}))
  n96 = Multiply(
   A=LinearInterpolate(
    A=ScalarParameter() {defaultValue=-0.1, parameterName=deactivated}
    B=ScalarParameter() {defaultValue=1, parameterName=activated}
    Alpha=@n49
   )
   B=Multiply(A=VectorParameter() {parameterName=GlowColor3_1, defaultValue=<rgba 0.6076 0 1 1>}.RGB, B=LinearInterpolate(A=@n86, B=@n92, Alpha=@n95))
  )
  n97 = Add(A=Multiply(A=@n66, B=@n69), B=Multiply(A=Power(Base=@n70, Exp=@n71), B=Multiply(A=@n69, B=@n96)))
  n98 = ScalarParameter() {parameterName=Zap}
  n99 = ScalarParameter() {parameterName=Zap2}
  n100 = Add(A=Add(A=Multiply(A=@n98) {constB=2}, B=Multiply(A=@n99) {constB=1.5}), B=@n49)
  n101 = LinearInterpolate(
   A=Multiply(A=ScalarParameter() {defaultValue=0.18, parameterName=InnerGlowWhileOff}, B=@n97)
   B=Multiply(A=@n97, B=Constant() {r=1})
   Alpha=@n100
  )
  n102 = ScalarParameter() {defaultValue=0.01, parameterName=InnerGlowWhileOff-Mobile}
  n103 = Add(A=@n66, B=Multiply(A=@n66, B=Multiply(A=@n92, B=@n95)))
  n104 = Constant() {r=1}
  n105 = QualitySwitch(Default=@n101, Low=LinearInterpolate(A=Multiply(A=@n102, B=@n103), B=Multiply(A=@n103, B=@n104), Alpha=@n100), High=@n101)
  n106 = Add(
   A=Multiply(A=@n39, B=VectorParameter() {group=SinkingGradient, parameterName=SinkGradColor, defaultValue=<rgba 0.8957 0 1.25 1>}.RGB)
   B=@n105
  )
  n107 = Multiply(A=@n1.R, B=Multiply(A=@n2, B=StaticSwitchParameter(True=@n106, False=@n105) {group=SinkingGradient, parameterName=UseSinkingGradient}))
  Clamp(None=@n107, Max=@n107)
  n108 = TextureCoordinate() {coordinateIndex=2}
  n109 = ScalarParameter() {defaultValue=0.8, parameterName=HLODRad}
  n110 = ScalarParameter() {defaultValue=1, parameterName=HLODDens}
  n111 = LinearInterpolate(
   A=VectorParameter() {parameterName=HLOD_Col0, defaultValue=<rgba 0 0 0 1>}.RGB
   B=VectorParameter() {parameterName=HLOD_Col1, defaultValue=<rgba 0.0746 0 0.1042 1>}.RGB
   Alpha=Clamp(None=MaterialFunctionCall(UVs=@n108, Radius=@n109, Density=@n110) {materialFunction=&RadialGradientExponential}.RadialGradientExponential)
  )
  LinearInterpolate(A=@n107, B=@n111, Alpha=ScalarParameter() {defaultValue=1, parameterName=tmep})
  MaterialFunctionCall!(TimeScale, MaxHeight, MinHeight) {materialFunction=&/MF_Creative_CoinBobble}
  MaterialProxyReplace(Realtime=@n107, MaterialProxy=@n111)
  MaterialProxyReplace(Realtime=@n35, MaterialProxy=Constant())
  MaterialProxyReplace(Realtime=Constant() {r=0.2}, MaterialProxy=Constant())
  MaterialProxyReplace(Realtime=@n14, MaterialProxy=Constant() {r=0.5})
  n112 = Constant() {r=1.4}
  n113 = Power(Base=Saturate(None=Multiply(A=@n9, B=@n112)), Exp=Constant() {r=1.5})
  MaterialProxyReplace(
   Realtime=LinearInterpolate(
    A=Constant() {r=0.29}
    B=Constant() {r=0.8}
    Alpha=MaterialFunctionCall(Transition Phase=@n113, Height Texture=@n12.R) {materialFunction=&HeightLerp}.Lerp Alpha No Contrast
   )
   MaterialProxy=Constant() {r=1}
  )
  n114 = Constant() {r=0.65}
  Multiply(A=Saturate(None=Add(A=@n12.R, B=@n114)), B=@n39)
 ]
}
