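; Creative coin prop device: a collectible that minigame players pick up by overlap to earn score, fire an On Collected event and send stat events.
; The server computes per-player collectibility and visibility from instance, team and class rules, replicating a HideFromPlayers list.
; Clients use that list to toggle mesh, idle FX and ambient audio.
; PickMesh async-loads the chosen mesh and material, queuing changes made mid-load, and upgrade paths migrate older options.
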(event OnPropertiesLoaded(ToyOptionsComponent)
  (bind (_mesh _scale _stattag) (Utilities|Struct|BreakCreativeDeviceMeshSettings (Utilities|Array|Get(aref) (Variables|Default|GetCoinMeshScaleTagArray) (Variables|Default|GetMeshIndex))))
  (bind _meshindex (Variables|Default|GetMeshIndex))
  (bind _coinmeshscaletagarray (Variables|Default|GetCoinMeshScaleTagArray))
  ;[1 1 1] Updating tag if property changed / loaded
  (Class|CreativeUIBroadcastComponent|Initialize (Variables|Default|GetCreative_HUDBroadcast_Component) (Variables|Default|GetFortMinigameProgress) (Variables|UserOptions|Advanced|GetDisplayScoreUpdateOnHUD))
  (CallFunction|PickMesh)
  ;;[1 0.2969 0.2969] Fix for FORT-477756. On Minigame Start is getting called before properties are load when loading directling into a match via an island code.
  (if (and (Class|CreativeCommonDeviceFunctionLibrary|IsinPlayMode self) (Variables|Default|GetStartGameDelegateFired))
    (CallFunction|CollectibleInitialized)
    (else
      _))
  (if (Utilities|Array|IsValidIndex _coinmeshscaletagarray _meshindex)
    (Networking|SwitchHasAuthority
      (:Authority
        (Variables|Default|SetCurrentCoinTag _stattag)
        (Initialization|InitializeComponentWithAffects (Variables|Default|GetCreative_UseRestriction_Component) (Variables|Default|GetFortMinigameProgress) 0 0 996 (Variables|UserOptions|GetClassIndex) 0 1 false false false)))))

(event OnAnyPropertyChanged(ToyOptionsComponent)
  (bind (_mesh _scale _stattag) (Utilities|Struct|BreakCreativeDeviceMeshSettings (Utilities|Array|Get(aref) (Variables|Default|GetCoinMeshScaleTagArray) (Variables|Default|GetMeshIndex))))
  (bind _meshindex (Variables|Default|GetMeshIndex))
  (bind _coinmeshscaletagarray (Variables|Default|GetCoinMeshScaleTagArray))
  ;[1 1 1] Updating tag if property changed / loaded
  (Class|CreativeUIBroadcastComponent|Initialize (Variables|Default|GetCreative_HUDBroadcast_Component) (Variables|Default|GetFortMinigameProgress) (Variables|UserOptions|Advanced|GetDisplayScoreUpdateOnHUD))
  (CallFunction|PickMesh)
  ;;[1 0.2969 0.2969] Fix for FORT-477756. On Minigame Start is getting called before properties are load when loading directling into a match via an island code.
  (if (and (Class|CreativeCommonDeviceFunctionLibrary|IsinPlayMode self) (Variables|Default|GetStartGameDelegateFired))
    (CallFunction|CollectibleInitialized)
    (else
      _))
  (if (Utilities|Array|IsValidIndex _coinmeshscaletagarray _meshindex)
    (Networking|SwitchHasAuthority
      (:Authority
        (Variables|Default|SetCurrentCoinTag _stattag)
        (Initialization|InitializeComponentWithAffects (Variables|Default|GetCreative_UseRestriction_Component) (Variables|Default|GetFortMinigameProgress) 0 0 996 (Variables|UserOptions|GetClassIndex) 0 1 false false false)))))

(event EventBeginPlay
  (bind (_min _max) (Components|StaticMesh|GetLocalBounds (Variables|Default|GetVisibleMesh)))
  (CallFunction|PickMesh)
  (Minigame|AddMinigameTag (Variables|Default|GetFortMinigameProgress) (Variables|Default|GetCurrentCoinTag))
  (Transformation|SetWorldLocation (Variables|Default|GetCollider) (+ (Transformation|GetWorldLocation (Variables|Default|GetVisibleMesh)) (/ (+ _min _max) 2.0)) false false)
  (Networking|CanExecuteCosmeticEvents
    (:True
      (CallFunction|UpdateCosmeticsforLocalPlayers)
      (Collision|SetCollisionResponsetoChannel (Variables|Default|GetCollider) "ECC_GameTraceChannel1" "ECR_Block")
      (Transformation|SetWorldScale3D (Variables|Default|GetCollider) (/ (Transformation|GetWorldScale (Variables|Debug|GetStaticMeshComponent)) 2.0)))))

(event Custom|Pick%20Mesh
  (bind (_min _max) (Components|StaticMesh|GetLocalBounds (Variables|Default|GetVisibleMesh)))
  (bind _meshindex (Variables|Default|GetMeshIndex))
  (bind _coinmeshscaletagarray (Variables|Default|GetCoinMeshScaleTagArray))
  (bind (_currentmeshscaletag_mesh _currentmeshscaletag_scale _currentmeshscaletag_stattag _currentmeshscaletag_material) (Utilities|Struct|BreakCreativeDeviceMeshSettings (Variables|GetCurrentMeshScaleTag_Mesh)))
  (bind (_currentmeshscaletag_mesh_2 _currentmeshscaletag_scale_2 _currentmeshscaletag_stattag_2) (Utilities|Struct|BreakCreativeDeviceMeshSettings (Variables|GetCurrentMeshScaleTag_Mesh)))
  (bind (_currentmeshscaletag_mesh_3 _currentmeshscaletag_scale_3) (Utilities|Struct|BreakCreativeDeviceMeshSettings (Variables|GetCurrentMeshScaleTag_Mesh)))
  (if (Variables|Default|GetbUpdatingMesh)
    (Variables|Default|SetbPendingMeshUpdate true)
    (elif (Utilities|Array|IsValidIndex _coinmeshscaletagarray _meshindex)
      (bind _output_get (Variables|Default|SetCurrentMeshScaleTag (Utilities|Array|Get(aref) _coinmeshscaletagarray _meshindex)))
      (Variables|Default|SetbUpdatingMesh true)
      (Variables|Default|SetCurrentCoinTag _currentmeshscaletag_stattag_2)
      (bind _loaded_object (Creative|Loading|AsyncLoadEditorSafe (Utilities|Struct|BreakCreativeDeviceMeshSettings _output_get))
        (:Loaded
          (bind _asstatic_mesh (Utilities|Casting|CastToStaticMesh _loaded_object))
          (Components|StaticMesh|SetStaticMesh (Variables|Default|GetVisibleMesh) _asstatic_mesh)
          (Collision|SetCollisionEnabled (Variables|Debug|GetStaticMeshComponent) "NoCollision")
          (Utilities|FlowControl|Sequence
            (:then_0
              (Transformation|SetRelativeScale3D (Variables|Default|GetVisibleMesh) _currentmeshscaletag_scale_3)
              (Transformation|SetWorldLocation (Variables|Default|GetCollider) (+ (Transformation|GetWorldLocation (Variables|Default|GetVisibleMesh)) (/ (+ _min _max) 2.0)) false false)
              (Networking|CanExecuteCosmeticEvents
                (:True
                  (CallFunction|UpdateCosmeticsforLocalPlayers)
                  (Collision|SetCollisionResponsetoChannel (Variables|Default|GetCollider) "ECC_GameTraceChannel1" "ECR_Block")
                  (Transformation|SetWorldScale3D (Variables|Default|GetCollider) (/ (Transformation|GetWorldScale (Variables|Debug|GetStaticMeshComponent)) 2.0)))))
            (:then_1
              ;[1 0.7441 0 0.5] Load Material once mesh has loaded
              (if (Utilities|IsValidSoftObjectReference _currentmeshscaletag_material)
                (bind _loaded_object_2 (Creative|Loading|AsyncLoadEditorSafe _currentmeshscaletag_material)
                  (:Loaded
                    (bind _asmaterial_interface (Utilities|Casting|CastToMaterialInterface _loaded_object_2)
                      (:then
                        (Rendering|Material|SetMaterial (Variables|Default|GetVisibleMesh) 0 _asmaterial_interface))
                      (:CastFailed
                        (Rendering|Material|SetMaterial (Variables|Default|GetVisibleMesh) 0 0)))
                    (Variables|Default|SetCurrentMeshComponentMaterial (Rendering|Material|GetMaterial (Variables|Default|GetVisibleMesh) 0))
                    (Variables|Default|SetbUpdatingMesh false)
                    (if (Variables|Default|GetbPendingMeshUpdate)
                      (Variables|Default|SetbPendingMeshUpdate false)
                      (CallFunction|PickMesh))))
                (else
                  _)))))))))

(event Building|EventOnBuildingActorInitialized
  (params (InitializationReason :type EFortBuildingInitializationReason%20Enum :default "StaticallyPlaced") (BuildingPersistentState :type EFortBuildingPersistentState%20Enum :default "Default"))
  ;[1 1 1] Turn off collision all day, everyday, no matter how you came into existence
  (Networking|SwitchHasAuthority
    (:Authority
      (Collision|SetCollisionEnabled (Variables|Debug|GetStaticMeshComponent) "NoCollision"))))

(event EventEndPlay
  (params (EndPlayReason :type EEndPlayReason%20Enum :default "Destroyed"))
  ;[1 1 1] Unbind team change
  (for _array_element (Fortnite|GetLocalFortPlayerControllers true true)
    (bind _asfort_player_state_athena (Utilities|Casting|CastToFortPlayerStateAthena (Variables|/Script/Engine.Controller|GetPlayerState _array_element)))
    (Player|UnbindallEventsfromOnTeamIndexChangedDelegate _asfort_player_state_athena)))

(event EventPostFinishSpawning
  ;[1 1 1] Turn off collision all day, everyday, no matter how you came into existence
  (|CallParentFunction)
  (CallFunction|MeshPickerUpgradePath)
  (CallFunction|OnlyVisibletoCollectingTeamUpgradePath)
  (Networking|SwitchHasAuthority
    (:Authority
      (Collision|SetCollisionEnabled (Variables|Debug|GetStaticMeshComponent) "NoCollision"))))

(event OnComponentBeginOverlap(Collider)
  (params (OverlappedComponent :type Primitive%20Component%20Object%20Reference) (OtherActor :type Actor%20Object%20Reference) (OtherComp :type Primitive%20Component%20Object%20Reference) (OtherBodyIndex :type Integer) (bFromSweep :type Boolean) (SweepResult :type Hit%20Result%20Structure%20%28by%20ref%29))
  (Networking|SwitchHasAuthority
    (:Authority
      (bind _currentminigame (Variables|/Script/FortniteGame.FortMinigameLogicComponent|GetCurrentMinigame (Variables|Default|GetFortMinigameProgress))
        (:then
          (Minigame|IsInProgress _currentminigame
            (:Yes
              (if (Variables|Default|GetbPossiblyVisibleInMinigame)
                (bind _showfx (CallFunction|DoCollection OtherActor))
                (if _showfx
                  (CallFunction|ShowFX))))))
        (:else)))))

(event MinigameEndedDelegate(FortMinigameProgress)
  (bind _minigame (Utilities|GetMinigame))
  ;[1 1 1] On Minigame Ended
  (Utilities|FlowControl|Sequence
    (:then_0
      (CallFunction|SetVisibility true))
    (:then_1
      (Collision|SetCollisionEnabled (Variables|Default|GetCollider) "QueryOnly")
      (Collision|SetCollisionResponsetoChannel (Variables|Default|GetCollider) "ECC_GameTraceChannel1" "ECR_Block"))
    (:then_2
      (Utilities|Set|Clear (Variables|Default|GetCollectedByPlayers))
      (Utilities|Map|Clear (Variables|Default|GetClientCoinVisibility))
      (Networking|SwitchHasAuthority
        (:Authority
          (Variables|Default|SetbPossiblyVisibleInMinigame true)
          (Utilities|Array|Clear (Variables|Default|GetHideFromPlayers)))))
    (:then_3
      (Components|Activation|SetActive (Variables|Default|GetIdleFX) true false)
      (CallFunction|UpdateAmbientAudio true)
      (Variables|Default|SetPickupFlash 0))
    (:then_4
      (Minigame|UnbindEventfromPlayerAddedDelegate _minigame (EventDispatchers|CreateEvent :!event Custom|On%20Player%20Added))
      (Minigame|UnbindEventfromPlayerRemovedDelegate _minigame (EventDispatchers|CreateEvent :!event Custom|On%20Player%20Removed))
      (Minigame|UnbindEventfromOnPlayerClassApplied _minigame (EventDispatchers|CreateEvent :!event Custom|On%20Player%20Class%20Changed))
      (Minigame|UnbindEventfromOnPlayerClassRemoved _minigame (EventDispatchers|CreateEvent :!event Custom|On%20Player%20Class%20Changed))
      (Minigame|UnbindEventfromOnPlayerJoininProgress _minigame (EventDispatchers|CreateEvent :!event Custom|On%20Player%20Joined%20In%20Progress))
      (Minigame|UnbindEventfromPlayerChangedTeamDelegate _minigame (EventDispatchers|CreateEvent :!event PlayerChangedTeam)))
    (:then_5
      (Variables|Default|SetbRequiresServerUpdate true))))

(event Custom|Show%20FX
  (bind _collider (Variables|Default|GetCollider))
  (bind _returnvalue (Transformation|GetWorldLocation _collider))
  ;[1 0.5253 0 0.2] Get sound and VFX to play on all clients
  (if (Math|Boolean|NORBoolean (Utilities|Enum|Equal(Enum) (Variables|UserOptions|GetShowPickupFX) "NewEnumerator1") (Utilities|Enum|Equal(Enum) (Variables|UserOptions|GetShowPickupFX) "NewEnumerator0"))
    (Networking|CanExecuteCosmeticEvents
      (:True
        (switch Utilities|FlowControl|Switch|SwitchonECreativeShowFX (Variables|UserOptions|GetShowPickupFX)
          (:NewEnumerator2
            _)
          (:NewEnumerator3
            (Effects|SpawnEmitteratLocation "/Game/Athena/Playlists/Score/Coins/FX/P_Athena_ScoreCoin_Collect_Gold.P_Athena_ScoreCoin_Collect_Gold" _returnvalue))
          (:NewEnumerator4
            (Utilities|FlowControl|Sequence
              (:then_0
                _)
              (:then_1
                _))))
        (Audio|PlaySoundatLocation "/Game/Athena/Sounds/Quests/ScoreCoin_Collect_Gold_Cue.ScoreCoin_Collect_Gold_Cue" _returnvalue)))))

(event EventDestroyed
  (bind _audio_looping (Variables|Default|GetAudio%20Looping))
  (Networking|CanExecuteCosmeticEvents
    (:True
      (if (Audio|Components|Audio|IsPlaying _audio_looping)
        (Audio|Components|Audio|FadeOut _audio_looping 0.5 0.0 "Linear" :!class /Script/Engine.AudioComponent)))))

(event Custom|On%20Player%20Joined%20In%20Progress
  (params (Pawn :type Pawn%20Object%20Reference))
  ;[1 1 1] Server Update
  (Variables|Default|SetbRequiresServerUpdate true)
  (Utilities|FlowControl|DelayUntilNextTick)
  (CallFunction|ServerUpdateforPlayers))

(event Custom|On%20Player%20Added
  (params (UniqueNetId :type Unique%20Net%20Id%20Structure) (bIsLocalPlayer :type Boolean))
  ;[1 1 1] Server Update
  (Variables|Default|SetbRequiresServerUpdate true)
  (Utilities|FlowControl|DelayUntilNextTick)
  (CallFunction|ServerUpdateforPlayers))

(event Custom|On%20Player%20Removed
  (params (UniqueNetId :type Unique%20Net%20Id%20Structure) (bIsLocalPlayer :type Boolean))
  ;[1 1 1] Server Update
  (Utilities|RemoveCollectorfromArray (Class|CreativeCommonDeviceFunctionLibrary|GetPlayerStateFromNetID UniqueNetId))
  (Utilities|FlowControl|DelayUntilNextTick)
  (CallFunction|ServerUpdateforPlayers))

(event Custom|On%20Player%20Class%20Changed
  (params (PlayerState :type Fort%20Player%20State%20Object%20Reference) (ClassSlotIndex :type Byte))
  ;[1 1 1] Server Update
  (Utilities|FlowControl|DelayUntilNextTick)
  (CallFunction|UpdateIndividualPlayerwithNotifyonChange PlayerState))

(event MinigameStartedDelegate(FortMinigameProgress)
  (bind _minigame (Utilities|GetMinigame))
  ;[1 1 1] On Minigame Started
  (Utilities|FlowControl|Sequence
    (:then_0
      (Networking|SwitchHasAuthority
        (:Authority
          (CallFunction|ServerUpdateforPlayers))))
    (:then_1
      (if (Variables|Debug|GetbSelfPickUp)
        (Utilities|Time|SetTimerbyEvent (delegate Custom|Timer) 15.0)))
    (:then_2
      (Variables|Default|SetbPossiblyVisibleInMinigame (Variables|Default|GetbVisibleOnMinigameStart)))
    (:then_3
      (Minigame|BindEventtoPlayerAddedDelegate _minigame (EventDispatchers|CreateEvent :!event Custom|On%20Player%20Added))
      (Minigame|BindEventtoPlayerRemovedDelegate _minigame (EventDispatchers|CreateEvent :!event Custom|On%20Player%20Removed))
      (Minigame|BindEventtoOnPlayerClassApplied _minigame (EventDispatchers|CreateEvent :!event Custom|On%20Player%20Class%20Changed))
      (Minigame|BindEventtoOnPlayerClassRemoved _minigame (EventDispatchers|CreateEvent :!event Custom|On%20Player%20Class%20Changed))
      (Minigame|BindEventtoOnPlayerJoininProgress _minigame (EventDispatchers|CreateEvent :!event Custom|On%20Player%20Joined%20In%20Progress))
      (Minigame|AssignPlayerChangedTeamDelegate _minigame (EventDispatchers|CreateEvent :!event PlayerChangedTeam)))
    (:then_4
      (Collision|SetCollisionResponsetoChannel (Variables|Default|GetCollider) "ECC_GameTraceChannel1" "ECR_Ignore"))))

(event Custom|Server%20Increment%20Pickup%20Flash
  ;[1 0 0.007519] This is only here to make the array 'HideForPlayers' replicate properly. This should be looked into
  (Variables|Default|SetPickupFlash (Math|Byte|Max(Byte) (Math|Byte|Byte+Byte (Variables|Default|GetPickupFlash) 1) 1)))

(event MinigameStateChangedDelegate(FortMinigameProgress)
  (params (Minigame :type Fort%20Minigame%20Object%20Reference) (NewMinigameState :type EFortMinigameState%20Enum :default "PreGame"))
  (bind _minigame (Utilities|GetMinigame))
  ;[1 1 1] On Minigame Started
  (if (Utilities|Enum|Equal(Enum) NewMinigameState "Warmup")
    (Utilities|FlowControl|Sequence
      (:then_0
        (Networking|SwitchHasAuthority
          (:Authority
            (CallFunction|ServerUpdateforPlayers))))
      (:then_1
        (if (Variables|Debug|GetbSelfPickUp)
          (Utilities|Time|SetTimerbyEvent (delegate Custom|Timer) 15.0)))
      (:then_2
        (Variables|Default|SetbPossiblyVisibleInMinigame (Variables|Default|GetbVisibleOnMinigameStart)))
      (:then_3
        (Minigame|BindEventtoPlayerAddedDelegate _minigame (EventDispatchers|CreateEvent :!event Custom|On%20Player%20Added))
        (Minigame|BindEventtoPlayerRemovedDelegate _minigame (EventDispatchers|CreateEvent :!event Custom|On%20Player%20Removed))
        (Minigame|BindEventtoOnPlayerClassApplied _minigame (EventDispatchers|CreateEvent :!event Custom|On%20Player%20Class%20Changed))
        (Minigame|BindEventtoOnPlayerClassRemoved _minigame (EventDispatchers|CreateEvent :!event Custom|On%20Player%20Class%20Changed))
        (Minigame|BindEventtoOnPlayerJoininProgress _minigame (EventDispatchers|CreateEvent :!event Custom|On%20Player%20Joined%20In%20Progress))
        (Minigame|AssignPlayerChangedTeamDelegate _minigame (EventDispatchers|CreateEvent :!event PlayerChangedTeam)))
      (:then_4
        (Collision|SetCollisionResponsetoChannel (Variables|Default|GetCollider) "ECC_GameTraceChannel1" "ECR_Ignore")))))

(event Custom|Timer
  ;[1 1 1] On Minigame Started
  (bind _showfx (CallFunction|DoCollection self))
  (if _showfx
    (CallFunction|ShowFX)))

(event Custom|CollectibleInitialized
  (bind _minigame (Utilities|GetMinigame))
  ;[1 1 1] On Minigame Started
  ;;[1 1 1] Make sure we update the collectible for all the players if they are initialized before device properties are loaded
  (Variables|Default|SetbRequiresServerUpdate true)
  (Utilities|FlowControl|Sequence
    (:then_0
      (Networking|SwitchHasAuthority
        (:Authority
          (CallFunction|ServerUpdateforPlayers))))
    (:then_1
      (if (Variables|Debug|GetbSelfPickUp)
        (Utilities|Time|SetTimerbyEvent (delegate Custom|Timer) 15.0)))
    (:then_2
      (Variables|Default|SetbPossiblyVisibleInMinigame (Variables|Default|GetbVisibleOnMinigameStart)))
    (:then_3
      (Minigame|BindEventtoPlayerAddedDelegate _minigame (EventDispatchers|CreateEvent :!event Custom|On%20Player%20Added))
      (Minigame|BindEventtoPlayerRemovedDelegate _minigame (EventDispatchers|CreateEvent :!event Custom|On%20Player%20Removed))
      (Minigame|BindEventtoOnPlayerClassApplied _minigame (EventDispatchers|CreateEvent :!event Custom|On%20Player%20Class%20Changed))
      (Minigame|BindEventtoOnPlayerClassRemoved _minigame (EventDispatchers|CreateEvent :!event Custom|On%20Player%20Class%20Changed))
      (Minigame|BindEventtoOnPlayerJoininProgress _minigame (EventDispatchers|CreateEvent :!event Custom|On%20Player%20Joined%20In%20Progress))
      (Minigame|AssignPlayerChangedTeamDelegate _minigame (EventDispatchers|CreateEvent :!event PlayerChangedTeam)))
    (:then_4
      (Collision|SetCollisionResponsetoChannel (Variables|Default|GetCollider) "ECC_GameTraceChannel1" "ECR_Ignore"))))

(event Custom|Turn%20Visibility%20On%20Event
  (params (Instigator :type Controller%20Object%20Reference))
  (Variables|Default|SetbPossiblyVisibleInMinigame true))

(event Custom|Turn%20Visibility%20Off%20Event
  (params (Instigator :type Controller%20Object%20Reference))
  (Variables|Default|SetbPossiblyVisibleInMinigame false))

(event Custom|Respawn%20Event
  (params (Instigator :type Player%20Controller%20Object%20Reference))
  (Utilities|RemoveCollectorfromArray (CallFunction|GetPlayerStatefromTargetActor Instigator))
  (CallFunction|ServerUpdateforPlayers))

(event Custom|Respawn%20For%20All%20Event
  (Utilities|Set|Clear (Variables|Default|GetCollectedByPlayers))
  (Variables|Default|SetbRequiresServerUpdate true)
  (CallFunction|ServerUpdateforPlayers))

(fn CanPlayerPickUp
  (params (FortPlayerState :type Fort%20Player%20State%20Object%20Reference))
  (bind _returnvalue (PlayerState|GetUniqueID FortPlayerState))
  (bind _clientcoinvisibility (Variables|Default|GetClientCoinVisibility))
  (bind (_value _returnvalue_2) (Utilities|Map|Find _clientcoinvisibility _returnvalue))
  (bind (_playerid_4_95a294f943dd14b55003f293fad0d8f7 _coincollectibility_19_33ff1a6246850380d3a27985748043b1 _coinvisibility_2_0aa632be4e5700a1fb276f90c334d421) (Utilities|Struct|BreakPlayerCoinVisibility _value))
  (Networking|SwitchHasAuthority
    (:Authority
      (if (Variables|Default|GetbPossiblyVisibleInMinigame)
        (Utilities|IsValid FortPlayerState
          (:Is%20Valid
            (if _returnvalue_2
              (return (and _coincollectibility_19_33ff1a6246850380d3a27985748043b1 _coinvisibility_2_0aa632be4e5700a1fb276f90c334d421))
              (else
                (return false))))
          (:Is%20Not%20Valid
            _))
        (else
          _))
      (return false))
    (:Remote
      _)))

(fn CheckClassRules
  (params (Player%20Actor :type Actor%20Object%20Reference))
  (bind _valid_controller (DEPRECATED|ActorHasValidController (Variables|Default|GetCreative_UseRestriction_Component) Player%20Actor true true))
  (return _valid_controller))

(fn CheckInstanceRules
  (params (CurrentPlayerId :type Unique%20Net%20Id%20Structure%20%28by%20ref%29) (CollectedTeamCounts :type Map%20of%20Integers%20to%20Integers) (CurrentPlayerTeam :type Integer%20%28by%20ref%29))
  (bind (_value _returnvalue) (Utilities|Map|Find (Variables|GetCollectedTeamCounts) (Variables|GetCurrentPlayerTeam)))
  (switch int (Variables|Default|GetInstanceRule)
    (:Default
      (return true))
    (:0
      (return (not (Utilities|Set|ContainsItem (Variables|Default|GetCollectedByPlayers) (Variables|GetCurrentPlayerId)))))
    (:1
      (return (<= (Utilities|Set|Length (Variables|Default|GetCollectedByPlayers)) 0)))
    (:2
      (return (<= (select _returnvalue _value 0) 0)))))

(fn CheckNonCollectingTeamVisibilityRules
  (params (CapturedPerTeam :type Map%20of%20Integers%20to%20Integers) (TotalPerTeam :type Map%20of%20Integers%20to%20Integers) (CurrentPlayerCollectingTeam :type Integer))
  (bind (_value _returnvalue) (Utilities|Map|Find (Variables|GetCapturedPerTeam) (CallFunction|GetCollectingTeam)))
  (bind _collecting_team (CallFunction|GetCollectingTeam))
  (bind _totalperteam (Variables|Default|GetTotalPerTeam))
  (bind (_value_2 _returnvalue_2) (Utilities|Map|Find _totalperteam _collecting_team))
  (bind _capturedperteam (Variables|GetCapturedPerTeam))
  (bind (_value_3 _returnvalue_3) (Utilities|Map|Find _capturedperteam _collecting_team))
  (if (and (Knobs|Team|IsValidTeamIndex (Variables|Default|GetCollectingTeam)) (!= (CallFunction|GetCollectingTeam) CurrentPlayerCollectingTeam))
    (switch int (Variables|Default|GetVisibleToNonCollectingTeam)
      (:Default
        _)
      (:0
        (return false))
      (:1
        (return true))
      (:2
        (switch int (Variables|Default|GetInstanceRule)
          (:Default
            (return true))
          (:0
            (if _returnvalue_3
              (if _returnvalue_2
                (return (< _value_3 _value_2))
                (else
                  (return false)))
              (else
                (return true))))
          (:1
            _)
          (:2
            _))
        (return (<= (select _returnvalue _value 0) 0))))
    (else
      (return false))))

(fn CheckTeamRules
  (params (CurrentPlayerTeam :type Integer))
  (return (or (Knobs|Team|IsTeamAll (Variables|Default|GetCollectingTeam)) (== (CallFunction|GetCollectingTeam) CurrentPlayerTeam))))

(fn DoCollection
  (params (ForActor :type Actor%20Object%20Reference))
  (bind _playerstateofcollector (LocalVariables|GetPlayerStateOfCollector))
  (bind _asfort_vehicle_interface (Utilities|Casting|CastToFortVehicleInterface ForActor)
    (:then
      (bind _playerstate (Utilities|GetPlayerStatefromActor (Vehicle|GetDriver _asfort_vehicle_interface))
        (:Is%20Valid
          (LocalVariables|SetPlayerStateOfCollector _playerstate)
          (if (not (CallFunction|CanPlayerPickUp (LocalVariables|GetPlayerStateOfCollector)))
            (return false)))
        (:Is%20Not%20Valid
          ;[1 1 1] Only used for debugging coin pick ups
          (if (not (and (Variables|Debug|GetbSelfPickUp) (Utilities|Equal(Object) ForActor self)))
            (return false)))) ; Get driver
      (Utilities|AddCollectortoArray _playerstateofcollector)
      (CallFunction|ServerUpdateforPlayers)
      (CallFunction|SendTrigger _playerstateofcollector)
      (CallFunction|SendStatEvents (LocalVariables|GetPlayerStateOfCollector))
      (return true))
    (:CastFailed
      (bind _playerstate_2 (Utilities|GetPlayerStatefromActor ForActor)
        (:Is%20Valid
          (LocalVariables|SetPlayerStateOfCollector _playerstate_2))
        (:Is%20Not%20Valid
          _)) ; Make sure it was a player
    )))

(fn GetPlayerStateFromTargetActor
  (params (Actor :type Actor%20Object%20Reference))
  (bind _returnvalue (Fortnite|GetFortPlayerControllerfromActor Actor))
  (bind _asfort_player_controller_spectating (Utilities|Casting|CastToFortPlayerControllerSpectating Actor)
    (:then
      (return (Class|FortPlayerControllerSpectating|GetFollowedPlayer _asfort_player_controller_spectating)))
    (:CastFailed
      (Utilities|IsValid _returnvalue
        (:Is%20Valid
          (bind _playerstate (Variables|/Script/Engine.Controller|GetPlayerState _returnvalue)
            (:then
              (bind _asfort_player_state (Utilities|Casting|CastToFortPlayerState _playerstate)
                (:then
                  (return _asfort_player_state))
                (:CastFailed
                  _)))
            (:else
              _))
          (return 0))
        (:Is%20Not%20Valid
          _)))))

(fn GetCollectingTeam
  (return (Math|Conversions|ToInteger(Byte) (Teams|IndextoTeam (Knobs|Team|GetTeamValueasIndex (Variables|Default|GetCollectingTeam))))))

(fn GetOverrideMeshMaterial
  (params (MeshComp :type Static%20Mesh%20Component%20Object%20Reference) (MatIdx :type Integer))
  (if (and (Utilities|Equal(Object) MeshComp (Variables|Default|GetVisibleMesh)) (== MatIdx 0))
    (return (Variables|Default|GetCurrentMeshComponentMaterial))
    (else
      (return (|CallParentFunction MeshComp MatIdx)))))

(fn IsTargetableByLinearMover
  (return true))

(fn MeshPickerUpgradePath
  (bind (_mesh _scale _stattag) (Utilities|Struct|BreakCreativeDeviceMeshSettings (Utilities|Array|Get(aref) (Variables|Default|GetCoinMeshScaleTagArray) (Variables|Default|GetMeshPicker))))
  (bind _meshpicker (Variables|Default|GetMeshPicker))
  (bind _coinmeshscaletagarray (Variables|Default|GetCoinMeshScaleTagArray))
  (if (Utilities|Array|IsValidIndex _coinmeshscaletagarray _meshpicker)
    (bind _output_get (LocalVariables|SetTempMeshPicker _meshpicker))
    (Variables|Default|SetCurrentCoinTag _stattag)
    (Variables|Default|SetMeshPicker -1)
    (OptionData|SetPropertyOverrideValues (Variables|Default|GetToyOptionsComponent) (Utilities|Map|MakeMap "MeshIndex" (Utilities|String|ToString(Integer) _output_get)))))

(fn MessageScore
  (return (Variables|UserOptions|Advanced|GetHUD%20Message) (Variables|UserOptions|Advanced|GetHUD%20MessageScoreColor) (Variables|UserOptions|Advanced|GetHUD%20MessageColor) (Variables|UserOptions|Advanced|GetResetHUD%20MessageScore)))

(fn OnlyVisibleToCollectingTeamUpgradePath
  (bind _toyoptionscomponent (Variables|Default|GetToyOptionsComponent))
  (bind (_currentoptions _returnvalue) (OptionData|GetPropertyOverrides _toyoptionscomponent))
  (bind (_value _returnvalue_2) (Utilities|Map|Find _currentoptions "bOnlyVisibleToCollectingTeam"))
  (bind _toyoptionscomponent_2 (Variables|Default|GetToyOptionsComponent))
  (Networking|SwitchHasAuthority
    (:Authority
      (if _returnvalue
        (if _returnvalue_2
          (bind _output_get (LocalVariables|SetCurrentVisibleTeamPropertyOverride _value))
          (bind _output_get_2 (LocalVariables|SetCurrentVisibleTeamActualValue (Utilities|String|Equal,CaseInsensitive(String) _output_get "TRUE")))
          (bind _output_get_3 (Variables|Default|SetVisibleToNonCollectingTeam (select _output_get_2 0 1)))
          (OptionData|SetPropertyOverrides _toyoptionscomponent_2 (Utilities|Map|MakeMap (Class|PlaylistUserOptionBase|GetOptionKey (Variables|Default|GetVisibleToNonCollectionTeamOption) false 0) (Class|PlaylistUserOptionBase|GetOptionValueNamefromValue (Variables|Default|GetVisibleToNonCollectionTeamOption) (Utilities|String|ToString(Integer) _output_get_3) true 0)))
          (OptionData|RemovePropertyOverrides _toyoptionscomponent_2 (Utilities|Array|MakeArray "bOnlyVisibleToCollectingTeam")))))))

(fn OnRep_bAmbientAudio
  (Networking|CanExecuteCosmeticEvents
    (:True
      (for _array_element (Fortnite|GetLocalFortPlayerControllers true true)
        (bind _out_unique_id (Utilities|GetUniqueIdfromActor _array_element)
          (:Is%20Valid
            (if (Utilities|Array|ContainsItem (Variables|Default|GetHideFromPlayers) _out_unique_id)
              (CallFunction|UpdateAmbientAudio false)
              (return)))))
      (CallFunction|UpdateAmbientAudio true))))

(fn OnRep_bPossiblyVisibleInMinigame
  (CallFunction|UpdateCosmeticsforLocalPlayers))

(fn OnRep_HideFromPlayers
  (CallFunction|UpdateCosmeticsforLocalPlayers))

(fn OnRep_MeshIndex
  (CallFunction|PickMesh))

(fn OnRep_MeshPicker
  (CallFunction|PickMesh))

(fn PlayerChangedTeam
  (params (PlayerState :type Fort%20Player%20State%20Object%20Reference) (NewTeam :type Byte) (OldTeam :type Byte) (ChangedWithoutRespawn :type Boolean))
  (if (!= NewTeam OldTeam)
    (Variables|Default|SetbRequiresServerUpdate true)
    (CallFunction|ServerUpdateforPlayers)))

(fn SendStatEvents
  (params (InFortPlayerState :type Fort%20Player%20State%20Object%20Reference))
  (bind (_outsourcetags _outcontexttags) (Class|FortMinigameStatsComponent|GetSourceandContextTags (LocalVariables|Getlocal%20MinigameStatsComp) (LocalVariables|GetFortPlayerController)))
  (bind _asfort_player_controller_zone (Utilities|Casting|CastToFortPlayerControllerZone (Actor|GetOwner InFortPlayerState)))
  (LocalVariables|SetFortPlayerController _asfort_player_controller_zone)
  (LocalVariables|GetFortPlayerController
    (:then
      (bind _output_get (LocalVariables|SetTargetTags (GameplayTags|GetOwnedGameplayTags self :!class /Script/GameplayTags.GameplayTagAssetInterface)))
      (GameplayTags|AddGameplayTag _output_get (Variables|Default|GetCurrentCoinTag) :!class /Script/GameplayTags.BlueprintGameplayTagLibrary)
      (Utilities|FlowControl|Sequence
        (:then_0
          (bind _output_get_2 (LocalVariables|SetQuestManager (Quests|GetQuestManager (LocalVariables|GetFortPlayerController) "Athena")))
          (Utilities|IsValid _output_get_2
            (:Is%20Valid
              (Quest|SendComplexCustomStatEvent (LocalVariables|GetQuestManager) self "(GameplayTags=)" (LocalVariables|GetTargetTags) (Variables|Default|GetAmountToCount)))))
        (:then_1
          (bind _returnvalue (CreativeDevices|GetMinigameStatsComp (LocalVariables|GetFortPlayerController) "Athena"))
          (LocalVariables|Setlocal%20MinigameStatsComp _returnvalue)
          (bind _local_minigamestatscomp (LocalVariables|Getlocal%20MinigameStatsComp)
            (:then
              (Class|FortMinigameStatsComponent|SendStatEventwithTags _local_minigamestatscomp (LocalVariables|GetFortPlayerController) "Collect" self (LocalVariables|GetTargetTags) (Class|FortMinigameStatsComponent|GetSourceandContextTags _local_minigamestatscomp (LocalVariables|GetFortPlayerController)) (.OutContextTags (Class|FortMinigameStatsComponent|GetSourceandContextTags _local_minigamestatscomp (LocalVariables|GetFortPlayerController))) (Variables|Default|GetAmountToCount))
              (GameplayTags|AddGameplayTag (LocalVariables|GetTargetTags) "(TagName=\"Minigame.Stat.Score.Coin\")" :!class /Script/GameplayTags.BlueprintGameplayTagLibrary)
              (Class|FortMinigameStatsComponent|SendStatEventwithTags (LocalVariables|Getlocal%20MinigameStatsComp) (LocalVariables|GetFortPlayerController) "MinigameDynamicEvent" self (LocalVariables|GetTargetTags) _outsourcetags _outcontexttags (Variables|Default|GetScoreValue))
              (Class|CreativeHUDBroadcastComponent|DoScoreEvent (Variables|Default|GetCreative_HUDBroadcast_Component) (LocalVariables|GetFortPlayerController) (Variables|Default|GetScoreValue) (Variables|UserOptions|Advanced|GetDisplayScoreUpdateOnHUD)))
            (:else)))))
    (:else)))

(fn SendTrigger
  (params (PlayerState :type Fort%20Player%20State%20Object%20Reference))
  (GameplayEvent|BroadcastEvent (Variables|UserOptions|Event|GetOn%20Collected) 0 (Utilities|Struct|MakeGameplayEventPickedUp (Utilities|Casting|CastToController (Actor|GetOwner PlayerState)))))

(fn ServerUpdateForPlayers
  (Networking|SwitchHasAuthority
    (:Authority
      (if (Variables|Default|GetbRequiresServerUpdate)
        (Utilities|Map|Clear (Variables|Default|GetCapturePerTeam))
        (Utilities|Map|Clear (Variables|Default|GetTotalPerTeam))
        (bind _currentminigame (Variables|/Script/FortniteGame.FortMinigameLogicComponent|GetCurrentMinigame (Variables|Default|GetFortMinigameProgress))
          (:then
            (bind _output_get (LocalVariables|SetParticipatingPlayers (Minigame|GetParticipatingPlayers _currentminigame)))
            (for _array_element _output_get
              (bind _output_get_2 (LocalVariables|SetCurrentPlayer _array_element))
              (Utilities|IsValid _output_get_2
                (:Is%20Valid
                  (LocalVariables|SetCurrentPlayerTeam (Math|Conversions|ToInteger(Byte) (Team|GetActorTeam _output_get_2)))
                  (bind _output_get_3 (LocalVariables|SetCurrentPlayerId (PlayerState|GetUniqueID (LocalVariables|GetCurrentPlayer))))
                  (LocalVariables|SetCurrentPlayerCaptured (Utilities|Set|ContainsItem (Variables|Default|GetCollectedByPlayers) _output_get_3))
                  (Utilities|IncrementTeamIntMap (Variables|Default|GetTotalPerTeam) (LocalVariables|GetCurrentPlayerTeam))
                  (if (LocalVariables|GetCurrentPlayerCaptured)
                    (Utilities|IncrementTeamIntMap (Variables|Default|GetCapturePerTeam) (LocalVariables|GetCurrentPlayerTeam))))))
            (Utilities|Map|Clear (Variables|Default|GetClientCoinVisibility))
            (LocalVariables|SetHideFromPlayersChanged false)
            (for _array_element_2 (LocalVariables|GetParticipatingPlayers)
              (bind _output_get_4 (LocalVariables|SetCurrentPlayer _array_element_2))
              (Utilities|IsValid _output_get_4
                (:Is%20Valid
                  (bind _output_get_5 (LocalVariables|SetCurrentPlayerTeam (Math|Conversions|ToInteger(Byte) (Team|GetActorTeam _output_get_4))))
                  (bind _output_get_6 (LocalVariables|SetCurrentPlayerId (PlayerState|GetUniqueID _output_get_4)))
                  (bind _hidefromplayerschanged (CallFunction|UpdateIndividualPlayer _output_get_6 _output_get_5 (LocalVariables|GetCurrentPlayer) (Variables|Default|GetHideFromPlayers)))
                  (if _hidefromplayerschanged
                    (LocalVariables|SetHideFromPlayersChanged true)))))
            (Variables|Default|SetbRequiresServerUpdate false)
            (if (LocalVariables|GetHideFromPlayersChanged)
              (Variables|Default|SetHideFromPlayers (Variables|Default|GetHideFromPlayers))))
          (:else))))))

(fn SetVisibility
  (params (bVisible :type Boolean))
  (Rendering|SetVisibility (Variables|Debug|GetStaticMeshComponent) bVisible false))

(fn ShouldInvisiblePropsBeAllowed
  (return false))

(fn UpdateAmbientAudio
  (params (bCanLocalPlayerPickup :type Boolean))
  (bind _audio_looping (Variables|Default|GetAudio%20Looping))
  (if (and bCanLocalPlayerPickup (Variables|Default|GetbAmbientAudio))
    (Audio|Components|Audio|Play _audio_looping 0.0)
    (else
      (Audio|Components|Audio|Stop _audio_looping))))

(fn UpdateCosmetics
  (params (IsVisible :type Boolean))
  (bind _bpossiblyvisibleinminigame (Variables|Default|GetbPossiblyVisibleInMinigame))
  (bind _returnvalue (and IsVisible _bpossiblyvisibleinminigame))
  (Networking|CanExecuteCosmeticEvents
    (:True
      (Utilities|FlowControl|Sequence
        (:then_1
          (CallFunction|SetVisibility _returnvalue)
          (Components|Activation|SetActive (Variables|Default|GetIdleFX) _returnvalue false))
        (:then_2
          (CallFunction|UpdateAmbientAudio _returnvalue))))))

(fn UpdateCosmeticsForLocalPlayers
  (for _array_element (Fortnite|GetLocalFortPlayerControllers true true)
    (bind _out_unique_id (Utilities|GetUniqueIdfromActor _array_element)
      (:Is%20Valid
        (if (Utilities|Array|ContainsItem (Variables|Default|GetHideFromPlayers) _out_unique_id)
          (CallFunction|UpdateCosmetics false)
          (return)))))
  (CallFunction|UpdateCosmetics true))

(fn UpdateIndividualPlayer
  (params (Player%20Id :type Unique%20Net%20Id%20Structure) (Player%20Team :type Integer) (Player%20State :type Fort%20Player%20State%20Object%20Reference) (Hide%20From%20Players :type Array%20of%20Unique%20Net%20Id%20Structures))
  (bind _new_collectibility (LocalVariables|GetNew%20Collectibility))
  (bind _player_id (Variables|GetPlayer%20Id))
  (bind _hide_from_players (Variables|GetHide%20From%20Players))
  (LocalVariables|SetNew%20Collectibility (Math|Boolean|ANDBoolean (CallFunction|CheckInstanceRules Player%20Id (Variables|Default|GetCapturePerTeam) (Variables|GetPlayer%20Team)) (Default|CheckTeamRules (Variables|GetPlayer%20Team)) (Default|CheckClassRules Player%20State)))
  (if _new_collectibility
    (LocalVariables|SetNew%20Visibility _new_collectibility)
    (else
      (LocalVariables|SetNew%20Visibility (CallFunction|CheckNonCollectingTeamVisibilityRules (Variables|Default|GetCapturePerTeam) (Variables|Default|GetTotalPerTeam) (Variables|GetPlayer%20Team)))))
  (Utilities|Map|Add (Variables|Default|GetClientCoinVisibility) (Variables|GetPlayer%20Id) (Utilities|Struct|MakePlayerCoinVisibility (Variables|GetPlayer%20Id) (LocalVariables|GetNew%20Collectibility) (LocalVariables|GetNew%20Visibility)))
  (if (LocalVariables|GetNew%20Visibility)
    (bind _returnvalue (Utilities|Array|RemoveItem _hide_from_players _player_id))
    (return _returnvalue)
    (else
      (bind _returnvalue_2 (Utilities|Array|AddUnique _hide_from_players _player_id))
      (return (!= _returnvalue_2 -1)))))

(fn UpdateIndividualPlayerWithNotifyOnChange
  (params (Player%20State :type Fort%20Player%20State%20Object%20Reference))
  (Networking|SwitchHasAuthority
    (:Authority
      (Utilities|IsValid Player%20State
        (:Is%20Valid
          (bind _hidefromplayerschanged (CallFunction|UpdateIndividualPlayer (PlayerState|GetUniqueID Player%20State) (Math|Conversions|ToInteger(Byte) (Team|GetActorTeam Player%20State)) Player%20State (Variables|Default|GetHideFromPlayers)))
          ;[1 1 1] Only Update Hide from Players if we've actually added to the array.
          (if _hidefromplayerschanged
            (Variables|Default|SetHideFromPlayers (Variables|Default|GetHideFromPlayers))))))))

(fn ConstructionScript
  (Variables|/Script/Engine.SceneComponent|SetbAbsoluteScale true (Variables|Default|GetIdleFX)))
